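# Physics asset for the gumballs in the Bubblegum backpack cosmetic, giving each ball a body that can collide inside the pack.
# Nineteen identical spheres represent the balls, while overlapping boxes on mstr_gumballs form an enclosure, with a ring of thin walls capped by two thick blocks.
# Capsules on the glass and rear limit joints add further colliders.
# Constraints tie balls to root or mstr_gumballs, and both limit joints to root, with free angular motion.
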PhysicsAsset /BRCosmetics/Accessories/FORT_Backpacks/F_MED_Bubblegum_Backpack/F_MED_Bubblegum_Pack_Balls_Physics
bodies {
 Glass_limit_jnt [
  {shapeType=Capsule, length=11, radius=6, shapeName=Glass_limit_jnt_capsule, center=<3.1236 0 2.694>, rotation=<rot 90 180 90>}
  {shapeType=Capsule, length=11, radius=6, shapeName=Glass_limit_jnt_capsule, center=<3.1799 0 9.27>, rotation=<rot 90 -165.9638 104.0362>}
 ]
 Rear_limit_jnt [{shapeType=Capsule, length=10, radius=5, shapeName=Rear_limit_jnt_capsule, center=<-1.7743 0 -0.3381>, rotation=<rot 90 180 90>}]
 dyn_gumball_01 [{shapeType=Sphere, radius=1.3}]
 dyn_gumball_02 [{shapeType=Sphere, radius=1.3}]
 dyn_gumball_03 [{shapeType=Sphere, radius=1.3}]
 dyn_gumball_04 [{shapeType=Sphere, radius=1.3}]
 dyn_gumball_05 [{shapeType=Sphere, radius=1.3}]
 dyn_gumball_06 [{shapeType=Sphere, radius=1.3}]
 dyn_gumball_07 [{shapeType=Sphere, radius=1.3}]
 dyn_gumball_08 [{shapeType=Sphere, radius=1.3}]
 dyn_gumball_09 [{shapeType=Sphere, radius=1.3}]
 dyn_gumball_10 [{shapeType=Sphere, radius=1.3}]
 dyn_gumball_11 [{shapeType=Sphere, radius=1.3}]
 dyn_gumball_12 [{shapeType=Sphere, radius=1.3}]
 dyn_gumball_13 [{shapeType=Sphere, radius=1.3}]
 dyn_gumball_14 [{shapeType=Sphere, radius=1.3}]
 dyn_gumball_15 [{shapeType=Sphere, radius=1.3}]
 dyn_gumball_16 [{shapeType=Sphere, radius=1.3}]
 dyn_gumball_17 [{shapeType=Sphere, radius=1.3}]
 dyn_gumball_18 [{shapeType=Sphere, radius=1.3}]
 dyn_gumball_19 [{shapeType=Sphere, radius=1.3}]
 mstr_gumballs [
  {shapeType=Box, extentX=10.0351, extentY=3.4304, extentZ=18.21, center=<8.6543 0.0291 0>, rotation=<rot 0 -85.6603 0>}
  {shapeType=Box, extentX=10.3196, extentY=3.2395, extentZ=18.4149, center=<-8.2461 -0.4702 0>, rotation=<rot 0 -85.6603 0>}
  {shapeType=Box, extentX=15.5674, extentY=3.186, extentZ=20.0215, center=<-0.1433 -0.1631 7.7006>, rotation=<rot 0 -85.6601 -90.3609>}
  {shapeType=Box, extentX=13.3655, extentY=2.5385, extentZ=16.151, center=<-0.0917 -0.0339 -8.2792>, rotation=<rot 0 -85.6601 -90.3609>}
  {shapeType=Box, extentX=10.7617, extentY=1.9034, extentZ=9.3272, center=<5.7194 -0.4587 -5.7027>, rotation=<rot 0 -85.6599 -141.6027>}
  {shapeType=Box, extentX=12.0819, extentY=2.002, extentZ=8.377, center=<-6.3192 -0.8401 -5.1827>, rotation=<rot 0.0001 -85.6597 140.4683>}
  {shapeType=Box, extentX=11.5662, extentY=2.4445, extentZ=10.3474, center=<5.6757 -0.387 5.2845>, rotation=<rot 0.0001 -85.6597 140.4683>}
  {shapeType=Box, extentX=10.7133, extentY=2.4263, extentZ=10.036, center=<-5.7702 -0.3527 5.7131>, rotation=<rot 0 -85.6599 -134.0959>}
  {shapeType=Box, extentX=31.6639, extentY=35.1589, extentZ=37.6676, center=<-0.4697 22.7559 0>}
  {shapeType=Box, extentX=34.1639, extentY=33.5757, extentZ=36.71, center=<0.4349 -20.1076 0>}
 ]
 root [{shapeType=Box, extentX=1.2384, extentY=0.8463, extentZ=1.1074, center=<0 -0.151 0>}]
}
constraints [
 {
  bone1Name=dyn_gumball_02
  bone2Name=root
  swing1LimitDegrees=45
  swing1Motion=Free
  swing2LimitDegrees=45
  swing2Motion=Free
  twistLimitDegrees=45
  twistMotion=Free
 }
 {
  bone1Name=dyn_gumball_01
  bone2Name=root
  swing1LimitDegrees=45
  swing1Motion=Free
  swing2LimitDegrees=45
  swing2Motion=Free
  twistLimitDegrees=45
  twistMotion=Free
 }
 {
  bone1Name=dyn_gumball_03
  bone2Name=root
  swing1LimitDegrees=45
  swing1Motion=Free
  swing2LimitDegrees=45
  swing2Motion=Free
  twistLimitDegrees=45
  twistMotion=Free
 }
 {
  bone1Name=dyn_gumball_07
  bone2Name=root
  swing1LimitDegrees=45
  swing1Motion=Free
  swing2LimitDegrees=45
  swing2Motion=Free
  twistLimitDegrees=45
  twistMotion=Free
 }
 {
  bone1Name=dyn_gumball_08
  bone2Name=root
  swing1LimitDegrees=45
  swing1Motion=Free
  swing2LimitDegrees=45
  swing2Motion=Free
  twistLimitDegrees=45
  twistMotion=Free
 }
 {
  bone1Name=dyn_gumball_09
  bone2Name=root
  swing1LimitDegrees=45
  swing1Motion=Free
  swing2LimitDegrees=45
  swing2Motion=Free
  twistLimitDegrees=45
  twistMotion=Free
 }
 {
  bone1Name=dyn_gumball_10
  bone2Name=root
  swing1LimitDegrees=45
  swing1Motion=Free
  swing2LimitDegrees=45
  swing2Motion=Free
  twistLimitDegrees=45
  twistMotion=Free
 }
 {
  bone1Name=dyn_gumball_05
  bone2Name=root
  swing1LimitDegrees=45
  swing1Motion=Free
  swing2LimitDegrees=45
  swing2Motion=Free
  twistLimitDegrees=45
  twistMotion=Free
 }
 {
  bone1Name=dyn_gumball_11
  bone2Name=mstr_gumballs
  swing1LimitDegrees=45
  swing1Motion=Free
  swing2LimitDegrees=45
  swing2Motion=Free
  twistLimitDegrees=45
  twistMotion=Free
 }
 {
  bone1Name=dyn_gumball_16
  bone2Name=mstr_gumballs
  swing1LimitDegrees=45
  swing1Motion=Free
  swing2LimitDegrees=45
  swing2Motion=Free
  twistLimitDegrees=45
  twistMotion=Free
 }
 {
  bone1Name=dyn_gumball_17
  bone2Name=mstr_gumballs
  swing1LimitDegrees=45
  swing1Motion=Free
  swing2LimitDegrees=45
  swing2Motion=Free
  twistLimitDegrees=45
  twistMotion=Free
 }
 {
  bone1Name=dyn_gumball_15
  bone2Name=mstr_gumballs
  swing1LimitDegrees=45
  swing1Motion=Free
  swing2LimitDegrees=45
  swing2Motion=Free
  twistLimitDegrees=45
  twistMotion=Free
 }
 {
  bone1Name=dyn_gumball_14
  bone2Name=mstr_gumballs
  swing1LimitDegrees=45
  swing1Motion=Free
  swing2LimitDegrees=45
  swing2Motion=Free
  twistLimitDegrees=45
  twistMotion=Free
 }
 {
  bone1Name=dyn_gumball_13
  bone2Name=mstr_gumballs
  swing1LimitDegrees=45
  swing1Motion=Free
  swing2LimitDegrees=45
  swing2Motion=Free
  twistLimitDegrees=45
  twistMotion=Free
 }
 {
  bone1Name=dyn_gumball_12
  bone2Name=mstr_gumballs
  swing1LimitDegrees=45
  swing1Motion=Free
  swing2LimitDegrees=45
  swing2Motion=Free
  twistLimitDegrees=45
  twistMotion=Free
 }
 {
  bone1Name=dyn_gumball_18
  bone2Name=mstr_gumballs
  swing1LimitDegrees=45
  swing1Motion=Free
  swing2LimitDegrees=45
  swing2Motion=Free
  twistLimitDegrees=45
  twistMotion=Free
 }
 {
  bone1Name=dyn_gumball_19
  bone2Name=mstr_gumballs
  swing1LimitDegrees=45
  swing1Motion=Free
  swing2LimitDegrees=45
  swing2Motion=Free
  twistLimitDegrees=45
  twistMotion=Free
 }
 {
  bone1Name=dyn_gumball_04
  bone2Name=mstr_gumballs
  swing1LimitDegrees=45
  swing1Motion=Free
  swing2LimitDegrees=45
  swing2Motion=Free
  twistLimitDegrees=45
  twistMotion=Free
 }
 {
  bone1Name=dyn_gumball_06
  bone2Name=mstr_gumballs
  swing1LimitDegrees=45
  swing1Motion=Free
  swing2LimitDegrees=45
  swing2Motion=Free
  twistLimitDegrees=45
  twistMotion=Free
 }
 {
  bone1Name=Glass_limit_jnt
  bone2Name=root
  swing1LimitDegrees=45
  swing1Motion=Free
  swing2LimitDegrees=45
  swing2Motion=Free
  twistLimitDegrees=45
  twistMotion=Free
 }
 {
  bone1Name=Rear_limit_jnt
  bone2Name=root
  swing1LimitDegrees=45
  swing1Motion=Free
  swing2LimitDegrees=45
  swing2Motion=Free
  twistLimitDegrees=45
  twistMotion=Free
 }
]
