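# Physics asset for the Bubblegum backpack cosmetic, built around its dangling fuse bone chain.
# Three identical box bodies form the fuse, chained from a smaller dyn_fuse_MSTR box through segments 01, 02 and 03.
# Each joint allows a narrow swing cone and an even tighter twist, so the fuse bends only slightly per segment.
# Small capsule bodies on the pack and mstr_gumballs bones have no constraints linking them to the chain.
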PhysicsAsset /BRCosmetics/Accessories/FORT_Backpacks/F_MED_Bubblegum_Backpack/F_MED_Bubblegum_Pack_Fuse_Physics
bodies {
 dyn_fuse_01 [{shapeType=Box, extentX=2.5067, extentY=1.0953, extentZ=1.4792, center=<0.3 0 0>}]
 dyn_fuse_02 [{shapeType=Box, extentX=2.5067, extentY=1.0953, extentZ=1.4792, center=<0.3 0 0>}]
 dyn_fuse_03 [{shapeType=Box, extentX=2.5067, extentY=1.0953, extentZ=1.4792, center=<0.3 0 0>}]
 dyn_fuse_MSTR [{shapeType=Box, extentX=0.9022, extentY=0.5565, extentZ=0.8179}]
 mstr_gumballs [{shapeType=Capsule, length=0, radius=1, center=<0 0 -3>}]
 pack [{shapeType=Capsule, length=1, radius=1}]
}
constraints [
 {bone1Name=dyn_fuse_02, bone2Name=dyn_fuse_01, swing1LimitDegrees=20, swing2LimitDegrees=20, twistLimitDegrees=5}
 {bone1Name=dyn_fuse_01, bone2Name=dyn_fuse_MSTR, swing1LimitDegrees=20, swing2LimitDegrees=20, twistLimitDegrees=5}
 {bone1Name=dyn_fuse_03, bone2Name=dyn_fuse_02, swing1LimitDegrees=20, swing2LimitDegrees=20, twistLimitDegrees=5}
]
