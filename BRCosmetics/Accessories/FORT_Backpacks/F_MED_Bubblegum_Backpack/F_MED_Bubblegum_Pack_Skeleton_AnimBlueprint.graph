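# Animation Blueprint for the Bubblegum backpack cosmetic, adding physics motion to its gumballs and fuse.
# The input pose runs through two chained RigidBody nodes with separate override physics assets, simulating in base bone space of mstr_gumballs and root respectively.
# Both nodes override world gravity and use scaled, clamped component motion.
# Half a second after skydiving ends, the event graph resets dynamics with a physics reset.
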AnimBlueprint /BRCosmetics/Accessories/FORT_Backpacks/F_MED_Bubblegum_Backpack/F_MED_Bubblegum_Pack_Skeleton_AnimBlueprint
target_skeleton=&/F_MED_Bubblegum_Pack_Skeleton
graphs [
 AnimGraph {nodes [
   AnimGraphNode_SubInput_0 = Input Pose()
   n1 = RigidBody(ComponentPose=Local To Component(LocalPose=@AnimGraphNode_SubInput_0)) {
    bClampLinearTranslationLimitToRefPose=true
    bOverrideWorldGravity=true
    overridePhysicsAsset=&/F_MED_Bubblegum_Pack_Balls_Physics
    simulationSpace=BaseBoneSpace
    alphaBoolBlend {blendOption=Linear}
    baseBoneRef {boneName=mstr_gumballs}
    componentAppliedLinearAccClamp=<5000 5000 5000>
    componentLinearAccScale=<0.41 0.41 0.41>
    componentLinearVelScale=<0.41 0.41 0.41>
    overrideWorldGravity {z=-980}
   }
   n2 = RigidBody(ComponentPose=@n1) {
    bClampLinearTranslationLimitToRefPose=true
    bOverrideWorldGravity=true
    overridePhysicsAsset=&/F_MED_Bubblegum_Pack_Fuse_Physics
    simulationSpace=BaseBoneSpace
    alphaBoolBlend {blendOption=Linear}
    baseBoneRef {boneName=root}
    componentAppliedLinearAccClamp=<5000 5000 5000>
    componentLinearAccScale=<0.5 0.5 0.6>
    componentLinearVelScale=<0.5 0.5 0.6>
    overrideWorldGravity {z=-980}
   }
   Output Pose(Result=Component To Local(ComponentPose=@n2)) {layerGroup=None}
  ]}
 EventGraph {nodes [
   Event OnEndSkydiving()
   > Delay(Duration=0.5)
   > ResetDynamics(InTeleportType=ResetPhysics)
   TryGetPawnOwner()
   Event BlueprintUpdateAnimation()
  ]}
]
variables []
